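; Quest gameplay ability that completes a drive-through-the-cornfield objective when the player drives a qualifying vehicle fast enough there.
; On the server, tag-add and tag-remove waits on the destination location tag start and clear a looping one-second speed check.
; The check requires matching vehicle tags and minimum forward speed, and prints the speed for debugging.
; Success stops the listeners, delays briefly to sell the damage, plays victory effects, sends the quest stat and ends.
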(event Ability|EventActivateAbilityFromEvent
  (params (EventData :type Gameplay%20Event%20Data%20Structure%20%28by%20ref%29))
  ;[1 1 1] Activate
  (|CallParentFunction EventData)
  (if (Networking|IsServer)
    (CallFunction|WaitforEnteringCornfield)))

(event Custom|Wait%20For%20Entering%20Cornfield
  (bind _destinationlocationtag (Variables|ReadOnly|GetDestinationLocationTag))
  ;[1 1 1] Entered / Left Cornfield
  (Utilities|FlowControl|Sequence
    (:then_0
      (bind _asynctaskproxy (Ability|Tasks|WaitGameplayTagAdd _destinationlocationtag)
        (:Added
          (Variables|Default|SetTagListenTask_Add _asynctaskproxy)
          (bind _returnvalue (Utilities|Time|SetTimerbyEvent (delegate Custom|Check%20Speed) 1.0 true)) ; Check vehicle speed in the cornfield
          (Variables|Default|SetVehicle%20Speed%20Check%20Timer%20Handle _returnvalue))) ; entered cornfield?
    )
    (:then_1
      (bind _asynctaskproxy_2 (Ability|Tasks|WaitGameplayTagRemove _destinationlocationtag)
        (:Removed
          (Variables|Default|SetTagListenTask_Remove _asynctaskproxy_2)
          (Utilities|Time|ClearandInvalidateTimerbyHandle (Variables|Default|GetVehicle%20Speed%20Check%20Timer%20Handle)))) ; left cornfield?
    )))

(event Custom|Check%20Speed
  ;[1 1 1] Ensure player is in a vehicle and driving fast enough
  ; to count for cornfield damage
  (bind _vehicle_occupied (Utilities|IsinVehicle (Variables|GAVars|Data|GetOwningPlayerPawn))
    (:true
      (Variables|Default|SetQuestVehicle _vehicle_occupied)
      (if (GameplayTags|HasAllTags (Class|FortVehicleInterface|GetGameplayTags _vehicle_occupied) (Variables|ReadOnly|GetRequiredVehicleTags))
        (bind _asfort_athena_vehicle (Utilities|Casting|CastToFortAthenaVehicle (Vehicle|GetVehicleInterface (Variables|GAVars|Data|GetOwningPlayerPawn))))
        (Development|PrintString (Utilities|String|ToString(Float) (Vehicle|GetForwardSpeedKmh _asfort_athena_vehicle))) ; debug speed
        (if (>= (Vehicle|GetForwardSpeedKmh _asfort_athena_vehicle) (Variables|Default|GetRequired%20Speed))
          (CallFunction|SuccessAndFinish))))))

(event Custom|VictoryFX
  ;[1 1 1] Victory FX
  (Audio|PlaySound2D (Variables|ReadOnly|GetVictorySound))
  (Effects|SpawnFXSystemAttached (Variables|ReadOnly|GetQuestSuccessEmitter) (Variables|/Script/Engine.Character|GetMesh (Variables|GAVars|Data|GetOwningPlayerPawn)) "Feet"))

(event Custom|Success%20and%20Finish
  ;[1 1 1] Success and Cleanup
  (GameplayTasks|EndTask (Variables|Default|GetTagListenTask_Add)) ; stop listening: enter cornfield
  (GameplayTasks|EndTask (Variables|Default|GetTagListenTask_Remove)) ; stop listening: left cornfield
  (Utilities|Time|ClearandInvalidateTimerbyHandle (Variables|Default|GetVehicle%20Speed%20Check%20Timer%20Handle)) ; stop watching for vehicle speed check
  (Utilities|FlowControl|Delay 3.0) ; give the player the illusion of driving\naround in the field for a few seconds so\nthey feel like they are doing some damage\nbefore the UI / VFX reward pops up
  (CallFunction|VictoryFX)
  (Quest|SendComplexCustomStatEvent (Variables|GAVars|Data|GetOwningPlayerQuestManager) 0 (Variables|Default|GetT_Quest) "(GameplayTags=)" 1)
  (Ability|EndAbility))
